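; Fortnite NPC ability where an AI pawn tosses a configured item as a world pickup.
; After starting a montage and briefly waiting, it spawns the pickup in front of the pawn.
; On success it broadcasts a custom quest verb for the converting player, optionally applies a quest-completion effect to itself, marks the pickup through that player's marker component and shows a speech bubble.
; It ends the ability after a delay.
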(event Ability|EventActivateAbility
  (bind _markercomponent (CallFunction|GetMarkerComponent))
  (bind (_montageinfo_montagetoplay _montageinfo_animplayrate _montageinfo_animrootmotiontranslationscale _montageinfo_montagesectiontoplay _montageinfo_overridesection) (Utilities|Struct|BreakFortGameplayAbilityMontageInfo (Variables|GetMontageInfo_MontageToPlay)))
  (bind _questcompletionge (Variables|Default|GetQuestCompletionGE))
  (|CallParentFunction)
  (Ability|Tasks|PlayMontageAndWaitNotify "None" _montageinfo_montagetoplay _montageinfo_animplayrate _montageinfo_overridesection true _montageinfo_animrootmotiontranslationscale
    (:then
      (Utilities|FlowControl|Delay 0.5)
      (bind _returnvalue (Inventory|SpawnPickupInWorld (Variables|Default|GetItemDef_ItemToGive) 1 (+ (* (Transformation|GetActorForwardVector (Variables|InternalData|GetAIPawn)) 1e+02) (Transformation|GetActorLocation (Variables|InternalData|GetAIPawn))) (* (Transformation|GetActorForwardVector (Variables|InternalData|GetAIPawn)) 1e+02) 0 true false false -1 "Tossed" "NPCGift"))
      (Utilities|IsValid _returnvalue
        (:Is%20Valid
          (CallFunction|BroadcastQuestAction)
          (if (Utilities|IsValidClass _questcompletionge)
            (GameplayEffects|ApplyGameplayEffectToSelf :self (Ability|GetAbilitySystemComponent (Variables|InternalData|GetAIPawn)) :GameplayEffectClass _questcompletionge :Level 0.0))
          (Utilities|IsValid _markercomponent
            (:Is%20Valid
              (MarkerComponent|MarkActoronServer _markercomponent _returnvalue true true))
            (:Is%20Not%20Valid
              _))
          (Class|BPFLAICommonFunctions|PlaySpeechBubble (Variables|InternalData|GetAIPawn) (Variables|Default|GetTag_GiveItemSpeechBubble) true true)
          (Utilities|FlowControl|Delay 3.0))
        (:Is%20Not%20Valid
          (Utilities|FlowControl|Delay 2.0 :!label _label)))
      (Ability|EndAbility))
    (:OnCompleted
      _)
    (:OnInterrupted
      _)
    (:OnCancelled
      _))
  (goto _label))

(event Custom|BroadcastQuestAction
  (bind _aipawn (Variables|InternalData|GetAIPawn))
  (bind _convertcomponent (Variables|/Script/FortniteGame.FortPlayerPawn|GetConvertComponent _aipawn))
  (bind _returnvalue (Convert|GetConverterPawn _convertcomponent))
  (bind _customverbtag (Variables|Default|GetCustomVerbTag))
  (Utilities|IsValid _convertcomponent
    (:Is%20Valid
      (Utilities|IsValid _returnvalue
        (:Is%20Valid
          (if (GameplayTags|IsGameplayTagValid _customverbtag)
            (CustomVerb|BroadcastCustomVerbMessage (CustomVerb|MakeCustomVerbMessage _returnvalue _customverbtag))))))))

(fn GetMarkerComponent
  (bind _aipawn (Variables|InternalData|GetAIPawn))
  (bind _returnvalue (Actor|GetComponentbyClass _aipawn "/Script/FortniteGame.FortPawnComponent_Convert"))
  (bind _returnvalue_2 (Convert|GetConverterPawn _returnvalue))
  (Utilities|IsValid _returnvalue
    (:Is%20Valid
      (Utilities|IsValid _returnvalue_2
        (:Is%20Valid
          (bind _asfort_player_controller_athena (Utilities|Casting|CastToFortPlayerControllerAthena (Fortnite|GetFortPlayerControllerfromActor _returnvalue_2))
            (:then
              (return (Class|FortPlayerController|GetMarkerComponent _asfort_player_controller_athena)))
            (:CastFailed
              (return 0))))
        (:Is%20Not%20Valid
          (return 0))))
    (:Is%20Not%20Valid
      (return 0))))
